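# Two-sided masked lit material for the pirate sail flag in the ARRGH skydiving contrail, usable on Niagara mesh particles.
# Panned flag-texture channels and world-position-driven sine height maps sum into a vertex offset, weighted by UV gradients or vertex color, mostly along normals.
# Niagara dynamic parameters adjust wave speed, tiling and time, while Fade Up pulls vertices vertically toward an offset pivot.
# Pirate main sail diffuse, mask and normal textures supply the surface look.
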Material /BRCosmetics/Athena/Effects/Materials/Skydiving/ARRGH/M_Arrgh_Flag
graph {
 properties {bUsedWithNiagaraMeshParticles=true, blendMode=BLEND_Masked, pixelDepthOffsetMode=PDOM_Legacy, twoSided=true, usedShadingModels=Default Lit}
 nodes [
  n1 = Panner(Coordinate=TextureCoordinate() {vTiling=0.5}) {bFractionalPart=true, speedY=1}
  n2 = ScalarParameter() {defaultValue=10, parameterName=Vertical Intensity}
  n3 = TextureCoordinate()
  n4 = ComponentMask(None=TextureCoordinate()) {r=true}
  n5 = Multiply(A=Sine(None=ComponentMask(None=@n3) {g=true}) {period=2}, B=Multiply(A=@n4, B=@n4))
  n6 = VertexNormalWS()
  n7 = Multiply(
   A=Multiply(
    A=TextureSample(UVs=@n1) {samplerSource=SSM_Wrap_WorldGroupSettings, texture=/BRCosmetics/Athena/Effects/Materials/Skydiving/ARRGH/Textures/T_Flag_M_Contrail}.G
    B=LinearInterpolate(B=@n2, Alpha=@n5)
   )
   B=@n6
  )
  n8 = Panner(
   Coordinate=MaterialFunctionCall(XY=TextureCoordinate() {vTiling=2}) {materialFunction=&Swizzle}.YX
   Time=Multiply(A=Time(), B=ScalarParameter() {defaultValue=-2, parameterName=WavySpeed})
  ) {bFractionalPart=true, speedY=1}
  n9 = ScalarParameter() {defaultValue=5, parameterName=Intensity}
  n10 = Power(Base=ComponentMask(None=TextureCoordinate()) {r=true}, Exp=ScalarParameter() {defaultValue=0.5, parameterName=HorizontaContrast})
  n11 = Multiply(
   A=@n6
   B=Multiply(
    A=TextureSample(UVs=@n8) {texture=&T_Flag_M_Contrail}.R
    B=LinearInterpolate(B=@n9, Alpha=@n10)
   )
  )
  n12 = VertexNormalWS()
  n13 = DynamicParameter() {defaultValue=<rgba 0 1 1 0>, paramNames [Speed, Tiling, Time Multi, Fade Up]}
  n14 = Add(A=Multiply(A=Time(), B=@n13.Time Multi), B=WorldPosition().XYZ)
  n15 = Multiply(A=ObjectPositionWS(), B=Add(A=ScalarParameter() {parameterName=Speed}, B=@n13.Speed)) {constB=0.95}
  n16 = Divide(A=ComponentMask(None=Subtract(A=@n14, B=@n15)) {b=true}, B=Add(A=ScalarParameter() {defaultValue=100, parameterName=Tilling}, B=@n13.Tiling))
  n17 = Multiply(A=Time(), B=ScalarParameter() {defaultValue=-2, parameterName=WavySpeed})
  n18 = TextureCoordinate()
  n19 = Multiply(
   A=TextureSample(UVs=Panner(Coordinate=@n16, Time=@n17) {bFractionalPart=true, speedX=1}) {samplerType=SAMPLERTYPE_Masks, texture=/Engine/Functions/Engine_MaterialFunctions02/ExampleContent/Textures/SineWaveHeightMap}.R
   B=LinearInterpolate(B=ScalarParameter() {defaultValue=10, parameterName=Intensity_2}, Alpha=ComponentMask(None=@n18) {r=true})
  )
  n20 = VectorParameter() {parameterName=Panner Speed, defaultValue=<rgba 1 0 0 0>}
  n21 = TextureSample(UVs=Panner(Coordinate=@n16, Speed=ComponentMask(None=@n20.RGB) {g=true, r=true})) {samplerType=SAMPLERTYPE_Masks, texture=&SineWaveHeightMap}
  n22 = VertexColor()
  n23 = MaterialFunctionCall() {materialFunction=&ObjectPivotPoint}
  n24 = Transform(None=VectorParameter() {parameterName=Pivot Offset for Flag, defaultValue=<rgba 0 0 35 1>}.RGB) {transformSourceType=TRANSFORMSOURCE_Local}
  n25 = Multiply(
   A=Subtract(A=Add(A=@n23.Object Pivot Location, B=@n24), B=WorldPosition().XYZ)
   B=DynamicParameter() {defaultValue=<rgba 0 1 1 0>, paramNames [Speed, Tiling, Time Multi, Fade Up]}.Fade Up
  )
  n26 = Constant3Vector() {constant=<rgba 0 0 1 0>}
  Add(A=Add(A=Add(A=@n7, B=@n11), B=Multiply(A=@n12, B=@n19)) {constB=0}, B=Add(A=Multiply(A=@n21.R, B=@n22.R), B=Multiply(A=@n25, B=@n26)))
  Multiply(
   A=TextureCoordinate()
   B=AppendVector(A=ScalarParameter() {defaultValue=2, parameterName=Large_UScale}, B=ScalarParameter() {defaultValue=1, parameterName=Large_VScale})
  )
  n27 = TextureCoordinate()
  n28 = AppendVector(A=ScalarParameter() {parameterName=Param_1}, B=ScalarParameter() {defaultValue=0.01})
  TextureSample(UVs=Clamp(None=Add(A=@n27, B=@n28))) {texture=/BRCosmetics/Athena/Effects/Materials/Skydiving/ARRGH/Textures/T_Pirate_Ship_Main_Sail_D_Contrail}
  TextureSample() {samplerType=SAMPLERTYPE_Masks, texture=/BRCosmetics/Athena/Effects/Materials/Skydiving/ARRGH/Textures/T_Pirate_Ship_Main_Sail_S_Contrail}
  n29 = TextureSample() {samplerType=SAMPLERTYPE_Normal, texture=/BRCosmetics/Athena/Effects/Materials/Skydiving/ARRGH/Textures/T_Pirate_Ship_Main_Sail_N_Contrail}
  Transform(None=MaterialFunctionCall(
    BaseNormal=@n29.RGB
    AdditionalNormal=MaterialFunctionCall() {materialFunction=&BlendAngleCorrectedNormals}.Result
   ) {materialFunction=&BlendAngleCorrectedNormals}.Result)
 ]
}
