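# Masked, two-sided lit material for the pirate ship rigging in the ARRGH skydiving contrail cosmetic, used on Niagara mesh particles.
# Diffuse, packed mask and normal textures supply the surface, with dimmed diffuse feeding an emissive adjustment function.
# A panned sine height map, mapped along world Z with time and object position, builds a wave.
# An inverted, squared pivot sphere mask and vertex red weight it; dynamic particle parameters adjust speed and tiling.
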Material /BRCosmetics/Athena/Effects/Materials/Skydiving/ARRGH/M_Arrgh_Rigging
graph {
 properties {bUsedWithNiagaraMeshParticles=true, blendMode=BLEND_Masked, pixelDepthOffsetMode=PDOM_Legacy, twoSided=true, usedShadingModels=Default Lit}
 nodes [
  ComponentMask(None=VectorParameter() {parameterName=Panner Speed, defaultValue=<rgba 1 0 0 0>}.RGB) {g=true, r=true}
  n1 = TextureSample() {samplerType=SAMPLERTYPE_Masks, texture=/BRCosmetics/Athena/Effects/Materials/Skydiving/ARRGH/Textures/T_Pirate_Ship_Rigging_S_Contrail}
  LinearInterpolate(Alpha=@n1.B) {constA=0.1, constB=0.9}
  MaterialFunctionCall(Emissive=Multiply(A=TextureSample() {texture=/BRCosmetics/Athena/Effects/Materials/Skydiving/ARRGH/Textures/T_Pirate_Ship_Rigging_D_Contrail}.RGB) {constB=0.2}) {materialFunction=&MF_FX_EmissiveAdjustments}
  n2 = OneMinus(None=SphereMask(A=WorldPosition().XYZ, B=ObjectPositionWS(), Radius=ScalarParameter() {defaultValue=150, parameterName=Sphere Msk size}, Hardness=Constant()))
  n3 = Multiply(A=@n2, B=@n2)
  n4 = DynamicParameter() {defaultValue=<rgba 0 1 1 0>, paramNames [Speed, Tiling, Time Multi, Fade Up]}
  n5 = Multiply(A=Time(), B=@n4.Time Multi)
  n6 = WorldPosition()
  n7 = ObjectPositionWS()
  n8 = Add(A=ScalarParameter() {parameterName=Speed}, B=@n4.Speed)
  n9 = ComponentMask(None=Subtract(A=Add(A=@n5, B=@n6.XYZ), B=Multiply(A=@n7, B=@n8) {constB=0.95})) {b=true}
  n10 = Add(A=ScalarParameter() {defaultValue=200, parameterName=Tilling}, B=@n4.Tiling)
  n11 = VectorParameter() {parameterName=Panner Speed, defaultValue=<rgba 1 0 0 0>}
  n12 = TextureSample(UVs=Panner(Coordinate=Divide(A=@n9, B=@n10), Speed=ComponentMask(None=@n11.RGB) {g=true, r=true})) {samplerType=SAMPLERTYPE_Masks, texture=/Engine/Functions/Engine_MaterialFunctions02/ExampleContent/Textures/SineWaveHeightMap}
  Multiply(A=Multiply(A=Multiply(A=@n3, B=@n12.R), B=ScalarParameter() {defaultValue=10, parameterName=Yup}), B=VertexColor().R)
  n13 = TextureSample() {samplerType=SAMPLERTYPE_Normal, texture=/BRCosmetics/Athena/Effects/Materials/Skydiving/ARRGH/Textures/T_Pirate_Ship_Rigging_N_Contrail}
  Reroute(None=@n13.RGB)
  Reroute(None=@n1.G)
  Reroute(None=@n1.R)
 ]
}
